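# Niagara swing-trail effect for the CrowbarSleek pickaxe, drawing short-lived gold and silver ribbons along the weapon.
# BladeTrailGold and BladeTrailSilver spawn anim-trail particles between different bone or socket pairs, updated from the User skeletal mesh.
# Both fade alpha over particle life through curves, with silver wider and fainter.
# A third Trail emitter, with time-of-day tinting and a hot color gradient, is disabled.
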NiagaraSystem /BRCosmetics/Weapons/FORT_Melee/Pickaxe_CrowbarSleek/FX/NS_Pickaxe_CrowbarSleek_Trails
dependencies {
 usedDataInterfaces [&NiagaraDataInterfaceCurve, &NiagaraDataInterfaceColorCurve]
 usedDynamicInputs []
 usedModules [&SystemState]
}
emitters {
 BladeTrailGold {
  emitterUpdateScript [
   &EmitterState()
   &FN_SpawnAnimTrail(Spawn Interval=0)
  ]
  particleSpawnScript [
   &FN_InitializeAnimTrailBoneIndecies(First Bone or Socket Index=1, Second Bone or Socket Index=0)
   &FN_InitializeAnimTrail(AnimTrailWidth=0.7, Color=<rgba 0.8 0.4916 0.16 1>, Lifetime=0.05)
  ]
  particleUpdateScript [
   &FN_UpdateAnimTrail(Locked Particle RibbonWidth=1, SkeletalMesh=$User.SkeletalMesh) {module=FN_UpdateAnimTrail001}
   &ScaleColor(
    Scale Alpha=&FloatFromCurve(
     FloatCurve=DataInterface() {propertyValues {Curve {keys [
         {arriveTangent=-2.003, interpMode=RCIM_Cubic, leaveTangent=-2.003, tangentMode=RCTM_User, time=0.5, value=1}
         {interpMode=RCIM_Cubic, tangentMode=RCTM_User, time=1, value=0}
        ]}}}
     Scale Curve=0.8
    )
    ScaleA=true
    ScaleRGB=true
   )
  ]
  rendererClasses [&NiagaraRibbonRendererProperties]
  renderers [{rendererClass=&NiagaraRibbonRendererProperties}]
 }
 BladeTrailSilver {
  emitterUpdateScript [
   &EmitterState()
   &FN_SpawnAnimTrail(Spawn Interval=0)
  ]
  particleSpawnScript [
   &FN_InitializeAnimTrailBoneIndecies(First Bone or Socket Index=2, Second Bone or Socket Index=0)
   &FN_InitializeAnimTrail(AnimTrailWidth=1.1, Color=<rgba 0.81 0.8359 0.9 1>, Lifetime=0.06)
  ]
  particleUpdateScript [
   &FN_UpdateAnimTrail(Locked Particle RibbonWidth=1, SkeletalMesh=$User.SkeletalMesh) {module=FN_UpdateAnimTrail001}
   &ScaleColor(
    Scale Alpha=&FloatFromCurve(
     FloatCurve=DataInterface() {propertyValues {Curve {keys [
         {arriveTangent=-2.003, interpMode=RCIM_Cubic, leaveTangent=-2.003, tangentMode=RCTM_User, time=0.5, value=1}
         {interpMode=RCIM_Cubic, tangentMode=RCTM_User, time=1, value=0}
        ]}}}
     Scale Curve=0.2
    )
    ScaleA=true
    ScaleRGB=true
   )
  ]
  rendererClasses [&NiagaraRibbonRendererProperties]
  renderers [{rendererClass=&NiagaraRibbonRendererProperties}]
 }
 Trail {
  bEnabled=false
  emitterUpdateScript [
   &EmitterState(Life Cycle Mode=::NewEnumerator1, Loop Behavior=::NewEnumerator1, Loop Duration=0.7, UseLoopDelay=false)
   &FN_SpawnAnimTrail_v2(AnimTrailWidthMode=::FromFirst, Spawn Interval=0)
  ]
  particleSpawnScript [
   /Game/Effects/Niagara/Modules/AnimTrail/FN_InitializeAnimTrail/FN_InitializeAnimTrailBoneIndecies(First Bone or Socket Index=2)
   /Game/Effects/Niagara/Modules/AnimTrail/FN_InitializeAnimTrail/FN_InitializeAnimTrail(
    AnimTrailWidth=2
    Color=&FN_Select_Color_from_Time_Of_Day(Morning=<rgba 1 1 1 0.1>, Day=<rgba 1 1 1 0.1>, Evening=<rgba 1 1 1 0.1>, Night=<rgba 0.5 0.5 0.5 0.01>)
    Lifetime=0.12
   )
  ]
  particleUpdateScript [
   &ScaleRibbonWidth(Ribbon Width Scale=&FloatFromCurve(FloatCurve=DataInterface() {propertyValues {Curve {keys [
         {arriveTangent=-2.4315, interpMode=RCIM_Cubic, leaveTangent=-2.4315, tangentMode=RCTM_User, time=0, value=1}
         {arriveTangent=0.0357, interpMode=RCIM_Cubic, leaveTangent=0.0357, tangentMode=RCTM_User, time=1, value=0}
        ]}}}))
   &ScaleColor(
    Scale Alpha=&Multiply_Float(
     A=&Multiply_Float(
      A=&FloatFromCurve(
       FloatCurve=DataInterface() {propertyValues {Curve {keys [
           {interpMode=RCIM_Cubic, leaveTangent=-5.8353, tangentMode=RCTM_Break, time=0, value=1.5}
           {arriveTangent=-0.3285, interpMode=RCIM_Cubic, leaveTangent=-0.3285, tangentMode=RCTM_User, time=0.6, value=0.15}
           {arriveTangent=-0.2139, interpMode=RCIM_Cubic, tangentMode=RCTM_Break, time=1, value=0}
          ]}}}
       Scale Curve=3
      )
      B=&FNTimeOfDayScalar(Night Scale=0.95, Morning Scale=0.3, Evening Scale=0.3)
     )
     B=&FloatFromCurve(
      FloatCurve=DataInterface() {propertyValues {Curve {keys [<at 0.1 0 RCIM_Cubic>,<at 0.16 1 RCIM_Cubic>]}}}
      CurveIndex=$Emitter.NormalizedLoopAge
     )
    )
    Scale RGB=&MakeVectorFromLinearColorRGB(Color=&ColorFromCurve(ColorCurve=DataInterface() {propertyValues {
        AlphaCurve {keys [<at 0 1 RCIM_Cubic>]}
        BlueCurve {keys [<at 0 0.08 RCIM_Cubic>,<at 0.0279 0>,<at 0.2274 0>,<at 1 0 RCIM_Cubic>]}
        GreenCurve {keys [<at 0 2.3493 RCIM_Cubic>,<at 0.0279 0.3138>,<at 0.2274 0.036>,<at 1 0 RCIM_Cubic>]}
        RedCurve {keys [<at 0 4 RCIM_Cubic>,<at 0.0279 1.6>,<at 0.2274 1>,<at 1 0 RCIM_Cubic>]}
       }}))
    ScaleA=true
    ScaleRGB=true
   ) {module=ScaleColor002}
   &FN_UpdateAnimTrail_v2()
   /SetVariables(Particles.RibbonWidthSecondary:NiagaraFloat=&Multiply_Float(A=$Particles.RibbonWidth, B=0.65)) {isSetParameters=true}
  ]
  rendererClasses [&NiagaraRibbonRendererProperties]
  renderers [{rendererClass=&NiagaraRibbonRendererProperties}]
 }
}
summary {
 systemName=NS_Pickaxe_CrowbarSleek_Trails
 emitters [
  {emitterName=BladeTrailSilver, rendererClasses [&NiagaraRibbonRendererProperties]}
  {bEnabled=false, emitterName=Trail, rendererClasses [&NiagaraRibbonRendererProperties]}
  {emitterName=BladeTrailGold, rendererClasses [&NiagaraRibbonRendererProperties]}
 ]
 userVariables [
  User.bisSwinging {type=NiagaraBool}
  User.Color A {type=LinearColor, description="", defaultValue {struct=LinearColor, value=<rgba 1 1 1 1>}}
  User.NewVector3f {type=Vector3f, description="", defaultValue {struct=Vector3f, value=<0 0 0>}}
  User.RibbonIDOffset {type=NiagaraInt32}
  User.SkeletalMesh {
   type=NiagaraDataInterfaceSkeletalMesh
   defaultValue {
    struct=NiagaraExt_VariableValue_DataInterface
    value {dataInterface=&self:NiagaraDataInterfaceSkeletalMesh_0}
   }
  }
 ]
}
